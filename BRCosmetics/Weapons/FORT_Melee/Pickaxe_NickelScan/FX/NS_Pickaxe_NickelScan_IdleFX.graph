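# Niagara idle effect for the NickelScan pickaxe cosmetic, with mesh swirls and a sprite decal pinned to the weapon's skeletal mesh.
# Top_Swirl, Bottom_Swirl and Arm_Swirl_Cuffs spawn continuously at offsets along the weapon, with cuffs partitioned between two positions; Decals burst-spawns.
# System variant and secondary colors tint the emitters.
# Top and bottom swirls have two mesh renderers and compute an extra narrower color layer, dark or cyan-tinted by User.VariantID.
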NiagaraSystem /BRCosmetics/Weapons/FORT_Melee/Pickaxe_NickelScan/FX/NS_Pickaxe_NickelScan_IdleFX
dependencies {
 usedDataInterfaces [
  &NiagaraDataInterfaceCurve
  &NiagaraDataInterfaceVector2DCurve
  /Script/Niagara.NiagaraDataInterfaceArrayBool
  &NiagaraDataInterfaceColorCurve
  &NiagaraDataInterfaceArrayFloat
  &NiagaraDataInterfaceArrayFloat3
  &NiagaraDataInterfaceSkeletalMesh
  &NiagaraDataInterfaceArrayColor
 ]
 usedDynamicInputs [&SelectLinearColorFromArray]
 usedModules [
  &self:SystemSpawnScript.SystemScriptSource.SystemScriptGraph.NiagaraNodeAssignment_0.SetVariables_#1
  &SystemState
  &self:SystemSpawnScript.SystemScriptSource.SystemScriptGraph.NiagaraNodeAssignment_2.SetVariables_#2
 ]
}
emitters {
 Arm_Swirl_Cuffs {
  emitterUpdateScript [
   &EmitterState()
   &SpawnRate(SpawnRate=3, Use Spawn Probability=false)
  ]
  particleSpawnScript [
   &PartitionParticles()
   &InitializeParticle(
    Color=$System.SecondaryColor
    Lifetime Mode=::NewEnumerator1
    Mesh Scale Mode=::NewEnumerator3
    Mesh Uniform Scale=0.5
    UsePositionOffset=false
    Write Material Random=true
   )
   &InitialMeshOrientation(
    Mesh Axis To Orient=<0 0 1>
    Mesh Orientation Mode=::NewEnumerator2
    Orientation Vector=<0 1 0>
    Rotation=&RandomRangeVector(Minimum=<0 -1 0>, Maximum=<0 1 0>)
    Use Rotation Vector=true
   )
   &CameraOffset(Camera Offset Amount=2)
  ]
  particleUpdateScript [
   &ParticleState()
   &ScaleMeshSize(Scale Factor=&VectorFromFloat(Value=&FloatFromCurve(
      FloatCurve=DataInterface() {propertyValues {Curve {keys [
          {interpMode=RCIM_Cubic, leaveTangent=0.3311, tangentMode=RCTM_Break, time=0.0008, value=0.4871}
          {arriveTangent=0.8171, interpMode=RCIM_Cubic, leaveTangent=0.8171, time=0.8, value=0.7085}
          {arriveTangent=7.6378, tangentMode=RCTM_User, time=0.9956, value=1.3}
         ]}}}
      Scale Curve=&SelectFloatFromArray(Direct Array Index=$Particles.PartitionParticles.Partition, Float Selection Array=DataInterface() {propertyValues {FloatData [1.1,0.9]}})
     )))
   &UpdateMeshOrientation(Rotation Rate=&RandomRangeFloat(Minimum=-1, Maximum=2)) {enabled=false}
   &SkeletalMeshLocation(
    MaskSpawnGroup=false
    Mesh Sampling Type=::NewEnumerator5
    Sampled Position Offset=&SelectVectorFromArray(
     Direct Array Index=$Particles.PartitionParticles.Partition
     Vector Selection Array=DataInterface() {propertyValues {FloatData [<0 0 -5>, <0 0 -113>]}}
    )
    Skeletal Mesh=$System.Skeletal Mesh
    Socket Sampling Mode=::NewEnumerator5
   )
   &ScaleColor(
    Scale Alpha=&FloatFromCurve(FloatCurve=DataInterface() {propertyValues {Curve {
        defaultValue=0
        keys [
         <at 0.0108 -0.0016 RCIM_Cubic>
         {interpMode=RCIM_Cubic, tangentMode=RCTM_User, time=0.3056, value=1}
         {interpMode=RCIM_Cubic, tangentMode=RCTM_User, time=1, value=0}
        ]
       }}})
    ScaleA=true
    ScaleRGB=true
   )
   &FN_Light_Attributes(Color=<rgba 1.0598 5 0 1>, Create Color Attribute=true, Create Exponent Attribute=true, Create Radius Attribute=true, Exponent=2, Radius=20)
   &DynamicMaterialParameters(
    Index 0 Param 1=1
    Index 3 Param 3=5
    Index 3 Param 4=5
    Module.Index0Float0Write=true
    Module.Index0Float1Write=true
    Module.Index0Float2Write=true
    Module.Index0Float3Write=true
    Module.Index3Float0Write=true
    Module.Index3Float1Write=true
    Module.Index3Float2Write=true
    Module.Index3Float3Write=true
    Write Parameter Index 3=::NewEnumerator1
   ) {enabled=false}
  ]
  rendererClasses [&NiagaraMeshRendererProperties]
  renderers [{rendererClass=&NiagaraMeshRendererProperties, rendererIndex=0}]
 }
 Bottom_Swirl {
  emitterUpdateScript [
   &EmitterState()
   &SpawnRate(SpawnRate=1.25, Use Spawn Probability=false)
  ]
  particleSpawnScript [
   &InitializeParticle(
    Color=$System.VariantColor
    Lifetime Mode=::NewEnumerator1
    Mesh Scale Max=<0.3 0.45 0.3>
    Mesh Scale Min=<0.2 0.4 0.2>
    Mesh Scale Mode=::NewEnumerator2
    UsePositionOffset=false
    Write Material Random=true
   )
   &InitialMeshOrientation(
    Mesh Orientation Mode=::NewEnumerator2
    Rotation=&RandomRangeVector(Minimum=<0 0 -0.5>, Maximum=<0 0 0.5>)
    Use Rotation Vector=true
   )
   &CameraOffset(Camera Offset Amount=3)
  ]
  particleUpdateScript [
   &ParticleState()
   &ScaleMeshSize(Scale Factor=&MakeVector(
     X=&FloatFromCurve(FloatCurve=DataInterface() {propertyValues {Curve {keys [
          {interpMode=RCIM_Cubic, leaveTangent=0.3311, tangentMode=RCTM_Break, time=0.0008, value=0.2353}
          {arriveTangent=0.9655, tangentMode=RCTM_User, time=0.9084, value=0.9333}
         ]}}})
     Y=1
     Z=1
    ))
   &UpdateMeshOrientation(Rotation Rate=&RandomRangeFloat(Minimum=100, Maximum=2), Rotation Vector=<0 1 0>) {enabled=false}
   &SkeletalMeshLocation(
    MaskSpawnGroup=false
    Mesh Sampling Type=::NewEnumerator5
    Sampled Position Offset=<0 0 -112>
    Skeletal Mesh=$System.Skeletal Mesh
    Socket Sampling Mode=::NewEnumerator5
   )
   &FN_Light_Attributes(Color=<rgba 1.0598 5 0 1>, Create Color Attribute=true, Create Exponent Attribute=true, Create Radius Attribute=true, Exponent=2, Radius=20) {enabled=false}
   /SetVariables(
    Particles.Black:LinearColor=&MultiplyLinearColor(
     Linear Color=$System.VariantColor
     Scale Factor=&MakeVector4_FromLinearColor(LinearColor=&MakeCustomLinearColorFromBool(
       Bool=/Niagara/DynamicInputs/Arrays/SelectBoolFromArray(
        Array Sampling Mode:ENiagara_LifetimeMode=::NewEnumerator0
        Direct Array Index:NiagaraInt32=&MakeIntFromFloat(Float=$User.VariantID)
        Direct Array Mode:ENiagara_WrapClamp=::NewEnumerator0
        Evaluation Type:ENiagaraRandomnessEvaluation=::NewEnumerator0
        Recalculate Random Each Loop:NiagaraBool=false
        Randomness Mode:ENiagaraRandomnessMode=::NewEnumerator0
        Random Seed:NiagaraInt32=?
        Fixed Random Seed:NiagaraBool=false
        Bool Selection Array:NiagaraDataInterfaceArrayBool=DataInterface() {propertyValues {GpuSyncMode=SyncCpuToGpu, MaxElements=0, BoolData [true,true,true,false,false,true]}}
       )
       True Color=&ColorFromCurve(ColorCurve=DataInterface() {propertyValues {
          AlphaCurve {keys [<at 0.0048 0>,<at 0.4333 1 RCIM_Cubic>,<at 1 0 RCIM_Cubic>]}
          BlueCurve {keys [<at 0.0021 0.2346>,<at 0.503 0.0317>]}
          GreenCurve {keys [<at 0.0021 0.2346>,<at 0.503 0.0317>]}
          RedCurve {keys [<at 0.0021 0.2346>,<at 0.503 0.0317>]}
         }})
       False Color=&ColorFromCurve(ColorCurve=DataInterface() {propertyValues {
          AlphaCurve {keys [<at 0.0048 0>,<at 0.4333 1 RCIM_Cubic>,<at 1 0 RCIM_Cubic>]}
          BlueCurve {keys [<at 0 1>]}
          GreenCurve {keys [<at 0 0.9492>]}
          RedCurve {keys [<at 0 0>]}
         }})
      ))
    )
    Particles.BlackPos:NiagaraPosition=&AddVectorToPosition(Position=$Output.SkeletalMeshLocation.SampledPositionWithoutOffset, Vector=<0 0 0>)
    Particles.BlackScale:Vector3f=&Multiply_Vector(A=$Particles.Scale, B=<0.5 1 1>)
   ) {isSetParameters=true, module=SetVariables_#3}
   &ScaleColor(
    Scale Mode=1
    Scale RGBA=&MakeVector4_FromLinearColor(LinearColor=&ColorFromCurve(ColorCurve=DataInterface() {propertyValues {
        AlphaCurve {keys [<at 0.0048 0>,<at 0.4333 1 RCIM_Cubic>,<at 1 0 RCIM_Cubic>]}
        BlueCurve {keys [<at 0 1 RCIM_Cubic>]}
        GreenCurve {keys [<at 0 1 RCIM_Cubic>]}
        RedCurve {keys [<at 0 1 RCIM_Cubic>]}
       }}))
    ScaleRGBA=true
   )
  ]
  rendererClasses [&NiagaraMeshRendererProperties]
  renderers [
   {rendererClass=&NiagaraMeshRendererProperties, rendererIndex=0}
   {rendererClass=&NiagaraMeshRendererProperties, rendererIndex=1}
  ]
 }
 Decals {
  emitterUpdateScript [
   &EmitterState(
    Inactive Response=::NewEnumerator1
    Life Cycle Mode=::NewEnumerator1
    Loop Behavior=::NewEnumerator1
    Loop Duration Mode=::NewEnumerator1
    UseLoopDelay=false
   )
   &SpawnBurst_Instantaneous(Use Spawn Probability=false, UseLoopCountLimit=false)
  ]
  particleSpawnScript [
   &InitializeParticle(Position Offset=<0 -10 0>, Sprite Size=<15 15>, Sprite Size Mode=::NewEnumerator4, UsePositionOffset=false, Write Material Random=true)
   &DynamicMaterialParameters(
    Index 0 Param 1=1
    Index 0 Param 2=1
    Index 1 Param 1=15
    Index 1 Param 2=30
    Module.Param0WriteEnabled=true
    Module.Param0WriteEnabled1=true
    Module.Param1WriteEnabled=true
    Module.Param1WriteEnabled1=true
    Module.Param2WriteEnabled=true
    Module.Param2WriteEnabled1=true
    Module.Param3WriteEnabled=true
    Module.Param3WriteEnabled1=true
    Write Parameter Index 1=true
   ) {enabled=false}
   &CameraOffset(Camera Offset Amount=15)
  ]
  particleUpdateScript [
   &ParticleState(
    Kill Particles When Lifetime Has Elapsed=false
    Let Infinitely Lived Particles Die When Emitter Deactivates=true
    Loop Particles Lifetime=true
   )
   &SkeletalMeshLocation(
    MaskSpawnGroup=false
    Mesh Sampling Type=::NewEnumerator5
    Sampled Position Offset=<0 0 -3>
    Skeletal Mesh=$System.Skeletal Mesh
    Socket Sampling Mode=::NewEnumerator5
   )
   &ScaleSpriteSize(
    Non-Uniform Curve Sprite Scale=DataInterface() {propertyValues {
      XCurve {keys [{tangentMode=RCTM_User, time=0, value=1}, {tangentMode=RCTM_User, time=0.75, value=1}]}
      YCurve {keys [{tangentMode=RCTM_User, time=0, value=0}, {tangentMode=RCTM_User, time=0.75, value=1}]}
     }}
    Scale Sprite Size Mode=::NewEnumerator3
   ) {enabled=false}
   &ScaleColor(Scale RGB=&MakeVectorFromLinearColorRGB(Color=$System.VariantColor), ScaleA=true, ScaleRGB=true)
   &FN_Light_Attributes(
    Color=$Particles.Color
    Create Color Attribute=true
    Create Position Attribute=true
    Create Radius Attribute=true
    Position=&AddVectorToPosition(Position=$StackContext.SkeletalMeshLocation.SampledPosition, Vector=<0 4 -15>)
    Radius=10
    Scale Color=5
   ) {enabled=false}
  ]
  rendererClasses [&NiagaraSpriteRendererProperties]
  renderers [{rendererClass=&NiagaraSpriteRendererProperties}]
 }
 Top_Swirl {
  emitterUpdateScript [
   &EmitterState()
   &SpawnRate(SpawnRate=2, Use Spawn Probability=false)
  ]
  particleSpawnScript [
   &InitializeParticle(
    Color=$System.VariantColor
    Lifetime Mode=::NewEnumerator1
    Mesh Scale Max=<0.5 0.6 0.6>
    Mesh Scale Min=<0.4 0.45 0.45>
    Mesh Scale Mode=::NewEnumerator2
    UsePositionOffset=false
    Write Material Random=true
   )
   &InitialMeshOrientation(
    Mesh Orientation Mode=::NewEnumerator2
    Rotation=&RandomRangeVector(Minimum=<0 0 -0.5>, Maximum=<0 0 0.5>)
    Use Rotation Vector=true
   )
   &CameraOffset(Camera Offset Amount=3)
  ]
  particleUpdateScript [
   &ParticleState()
   &ScaleMeshSize(Scale Factor=&MakeVector(
     X=&FloatFromCurve(FloatCurve=DataInterface() {propertyValues {Curve {keys [
          {interpMode=RCIM_Cubic, leaveTangent=0.3311, tangentMode=RCTM_Break, time=0.0008, value=0.2353}
          {arriveTangent=0.9655, tangentMode=RCTM_User, time=0.9084, value=0.9333}
         ]}}})
     Y=1
     Z=1
    ))
   &UpdateMeshOrientation(Rotation Rate=&RandomRangeFloat(Minimum=100, Maximum=2), Rotation Vector=<0 1 0>) {enabled=false}
   &SkeletalMeshLocation(MaskSpawnGroup=false, Mesh Sampling Type=::NewEnumerator5, Skeletal Mesh=$System.Skeletal Mesh, Socket Sampling Mode=::NewEnumerator5)
   &FN_Light_Attributes(Color=<rgba 1.0598 5 0 1>, Create Color Attribute=true, Create Exponent Attribute=true, Create Radius Attribute=true, Exponent=2, Radius=20) {enabled=false}
   /SetVariables(
    Particles.Black:LinearColor=&MultiplyLinearColor(
     Linear Color=$System.VariantColor
     Scale Factor=&MakeVector4_FromLinearColor(LinearColor=&MakeCustomLinearColorFromBool(
       Bool=&SelectBoolFromArray(
        Array Sampling Mode:ENiagara_LifetimeMode=::NewEnumerator0
        Direct Array Index:NiagaraInt32=&MakeIntFromFloat(Float=$User.VariantID)
        Direct Array Mode:ENiagara_WrapClamp=::NewEnumerator0
        Evaluation Type:ENiagaraRandomnessEvaluation=::NewEnumerator0
        Recalculate Random Each Loop:NiagaraBool=false
        Randomness Mode:ENiagaraRandomnessMode=::NewEnumerator0
        Random Seed:NiagaraInt32=?
        Fixed Random Seed:NiagaraBool=false
        Bool Selection Array:NiagaraDataInterfaceArrayBool=DataInterface() {propertyValues {GpuSyncMode=SyncCpuToGpu, MaxElements=0, BoolData [true,true,true,false,false,true]}}
       )
       True Color=&ColorFromCurve(ColorCurve=DataInterface() {propertyValues {
          AlphaCurve {keys [<at 0.0048 0>,<at 0.4333 1 RCIM_Cubic>,<at 1 0 RCIM_Cubic>]}
          BlueCurve {keys [<at 0.0021 0.2346>,<at 0.503 0.0317>]}
          GreenCurve {keys [<at 0.0021 0.2346>,<at 0.503 0.0317>]}
          RedCurve {keys [<at 0.0021 0.2346>,<at 0.503 0.0317>]}
         }})
       False Color=&ColorFromCurve(ColorCurve=DataInterface() {propertyValues {
          AlphaCurve {keys [<at 0.0048 0>,<at 0.4333 1 RCIM_Cubic>,<at 1 0 RCIM_Cubic>]}
          BlueCurve {keys [<at 0 1>]}
          GreenCurve {keys [<at 0 0.9492>]}
          RedCurve {keys [<at 0 0>]}
         }})
      ))
    )
    Particles.BlackPos:NiagaraPosition=&AddVectorToPosition(Position=$Output.SkeletalMeshLocation.SampledPositionWithoutOffset, Vector=<0 0 0>)
    Particles.BlackScale:Vector3f=&Multiply_Vector(A=$Particles.Scale, B=<0.5 1 1>)
   ) {isSetParameters=true, module=SetVariables_#3}
   &ScaleColor(
    Scale Mode=1
    Scale RGBA=&MakeVector4_FromLinearColor(LinearColor=&ColorFromCurve(ColorCurve=DataInterface() {propertyValues {
        AlphaCurve {keys [<at 0.0048 0>,<at 0.4333 1 RCIM_Cubic>,<at 1 0 RCIM_Cubic>]}
        BlueCurve {keys [<at 0 1 RCIM_Cubic>]}
        GreenCurve {keys [<at 0 1 RCIM_Cubic>]}
        RedCurve {keys [<at 0 1 RCIM_Cubic>]}
       }}))
    ScaleRGBA=true
   )
  ]
  rendererClasses [&NiagaraMeshRendererProperties]
  renderers [
   {rendererClass=&NiagaraMeshRendererProperties, rendererIndex=0}
   {rendererClass=&NiagaraMeshRendererProperties, rendererIndex=1}
  ]
 }
}
summary {
 systemName=NS_Pickaxe_NickelScan_IdleFX
 emitters [
  {emitterName=Decals, rendererClasses [&NiagaraSpriteRendererProperties]}
  {emitterName=Top_Swirl, rendererClasses [&NiagaraMeshRendererProperties]}
  {emitterName=Arm_Swirl_Cuffs, rendererClasses [&NiagaraMeshRendererProperties]}
  {emitterName=Bottom_Swirl, rendererClasses [&NiagaraMeshRendererProperties]}
 ]
 userVariables [
  User.bisFootHidden {type=NiagaraBool}
  User.bIsFrontEnd {type=NiagaraBool}
  User.bIsFrontEndPreview {type=NiagaraBool}
  User.VariantID {type=NiagaraFloat, defaultValue=0, description=""}
 ]
}
